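# Parent character material for the Gallium Reality outfit that prints a loading screen image onto the body.
# Masks from UV integer offsets or spheres select rear, sleeve and shoe regions with their own transforms, offsetting the left shoe by pre-skinned X sign.
# The masked, optionally clamped image composites over background and pattern colors, then multiplies gamma-lifted base color.
# The result blends into material attributes between MF_CharacterParent_2023 PreFX and PostFX.
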Material /BRCosmetics/Characters/Player/Male/Medium/Bodies/M_MED_GalliumReality/Materials/M_Gallium_Parent
graph {
 properties {
  bUseMaterialAttributes=true
  bUsedWithClothing=true
  bUsedWithInstancedStaticMeshes=true
  bUsedWithMeshDeformer=true
  bUsedWithMorphTargets=true
  bUsedWithNanite=true
  bUsedWithNiagaraMeshParticles=true
  bUsedWithSkeletalMesh=true
  pixelDepthOffsetMode=PDOM_Legacy
  shadingModel=MSM_FromMaterialExpression
  usedShadingModels=Default Lit | Subsurface | Clear Coat | Subsurface Profile | Cloth
 }
 nodes [
  n1 = NamedRerouteUsage()
  n2 = NamedRerouteUsage()
  Add(A=Multiply(A=Divide(A=@n1) {constB=100}, B=Constant2Vector() {g=-1, r=1}), B=@n2)
  n3 = NamedRerouteUsage()
  n4 = NamedRerouteUsage()
  n5 = StaticSwitchParameter(True=Multiply(A=OneMinus(None=@n3), B=@n4), False=@n4) {defaultValue=true, group=Gallium, parameterName=Enable Sleeves, sortPriority=1}
  n6 = ScalarParameter() {defaultValue=100, group=Gallium, parameterName=LoadingScreenScale_Rear, sliderMax=200, sliderMin=50, sortPriority=31}
  n7 = MaterialFunctionCall(
   X=ScalarParameter() {group=Gallium, parameterName=LoadingScreenOffsetX_Rear, sliderMax=100, sortPriority=31}
   Y=ScalarParameter() {group=Gallium, parameterName=LoadingScreenOffsetY_Rear, sliderMax=100, sortPriority=31}
   Z=ScalarParameter() {group=Gallium, parameterName=LoadingScreenRotation_Rear, sliderMax=360, sortPriority=31}
   A=@n6
  ) {materialFunction=&MakeFloat4}
  n8 = MaterialFunctionCall(
   X=ScalarParameter() {group=Gallium, parameterName=LoadingScreenOffsetX, sliderMax=100, sortPriority=30}
   Y=ScalarParameter() {group=Gallium, parameterName=LoadingScreenOffsetY, sliderMax=100, sortPriority=30}
   Z=ScalarParameter() {group=Gallium, parameterName=LoadingScreenRotation, sliderMax=360, sortPriority=30}
   A=ScalarParameter() {defaultValue=100, group=Gallium, parameterName=LoadingScreenScale, sliderMax=200, sliderMin=50, sortPriority=30}
  ) {materialFunction=&MakeFloat4}
  n9 = StaticSwitchParameter(
   True=If(
    A=ScalarParameter() {group=Gallium, parameterName=EnableSeparateRearUVTransform}
    "A > B"=If(A=@n5, "A > B"=@n7.Result, "A < B"=@n8.Result) {constB=0.5}
    "A < B"=@n8.Result
   ) {constB=0.5}
   False=@n8.Result
  ) {defaultValue=true, group=Gallium, parameterName=Enable Rear, sortPriority=1}
  n10 = StaticSwitchParameter(True=If(A=NamedRerouteUsage(), "A > B"=Constant4Vector() {constant=<rgba 0 0 0 100>}.RGBA, "A < B"=@n9) {constB=0.5}, False=@n9) {defaultValue=true, group=Gallium, parameterName=Enable Shoes, sortPriority=1}
  ComponentMask(None=@n10) {g=true, r=true}
  ComponentMask(None=@n10) {a=true}
  n11 = NamedRerouteUsage()
  n12 = NamedRerouteUsage()
  n13 = ScalarParameter() {
   defaultValue=0.25
   desc=To brighten up the basecolor below the loading screen before multiplying it
   group=Gallium
   parameterName=LoadingScreen_BackgroundGamma
  }
  n14 = VectorParameter() {group=Gallium, parameterName=BackgroundColor, sortPriority=30, defaultValue=<rgba 0.3333 0 0.0515 1>}
  n15 = VectorParameter() {group=Gallium, parameterName=PatternColor, sortPriority=30, defaultValue=<rgba 0.3591 0.5264 1 1>}
  n16 = TextureSampleParameter2D!(UVs, Bias=Constant() {r=-3}, Apply View MipBias) {
   group=Gallium
   mipValueMode=TMVM_MipBias
   parameterName=MaskTexture
   samplerSource=SSM_Wrap_WorldGroupSettings
   samplerType=SAMPLERTYPE_Masks
   sortPriority=4
   texture=/BRCosmetics/Characters/Player/Male/Medium/Bodies/M_MED_GalliumReality/Textures/T_GalliumReality_Body_Mask_01
  }
  n17 = StaticSwitchParameter(True=Min(A=OneMinus(None=@n16.G), B=@n16.B), False=@n16.B) {group=Gallium, parameterName=AlwaysFillBackgroundColorBelowLoadingScreen}
  n18 = NamedRerouteUsage()
  n19 = TextureSampleParameter2D!(UVs=@n18, Apply View MipBias) {
   group=Gallium
   parameterName=LoadingScreenOverride
   samplerSource=SSM_Wrap_WorldGroupSettings
   sortPriority=4
   texture=/BRCosmetics/2dAssets/Loadingscreens/Season30/T_LS_S30_GalliumConcrete
  }
  n20 = TextureSampleParameter2D(UVs=@n18) {
   group=Gallium
   parameterName=LoadingScreen
   samplerSource=SSM_Wrap_WorldGroupSettings
   sortPriority=4
   texture=&T_LS_S30_GalliumConcrete
  }
  n21 = Constant() {r=0.5}
  n22 = NamedRerouteDeclaration(None=If(
    A=ScalarParameter() {group=Gallium, parameterName=LoadingScreenUVChannel, sortPriority=28}
    "A > B"=TextureCoordinate() {coordinateIndex=1}
    "A < B"=TextureCoordinate() {coordinateIndex=2}
   ) {constB=0.5}) {name=RawUV, nodeColor=<rgba 0.5234 1 0 1>}
  n23 = StaticSwitchParameter(True=If(A=NamedRerouteUsage(), "A > B"=NamedRerouteUsage(), "A < B"=@n22) {constB=0.5}, False=@n22) {defaultValue=true, group=Gallium, parameterName=Enable Shoes, sortPriority=1}
  n24 = Divide(A=NamedRerouteUsage()) {constB=360}
  n25 = Add(A=@n24, B=NamedRerouteUsage())
  n26 = MaterialFunctionCall() {materialFunction=&Pi}
  n27 = Multiply(A=NamedRerouteUsage(), B=Constant2Vector() {g=0.5, r=1}) {desc=aspect ratio}
  n28 = MaterialFunctionCall(UVs=Rotator(Coordinate=Frac(None=@n23), Time=Multiply(A=@n25, B=@n26.Result)) {speed=-2}, Texture Scale=@n27) {materialFunction=&ScaleUVsByCenter}
  n29 = MaterialFunctionCall(Float2=NamedRerouteDeclaration(None=Subtract(A=@n28.UVs, B=@n2)) {name=ImageUV, nodeColor=<rgba 0.3594 1 0 1>}) {materialFunction=&BreakOutFloat2Components}
  n30 = Multiply(A=SphereMask(A=@n21, B=@n29.R) {attenuationRadius=0.5}, B=SphereMask(A=@n29.G, B=@n21) {attenuationRadius=0.5})
  n31 = Saturate(None=Add(A=@n30, B=NamedRerouteUsage()))
  n32 = LinearInterpolate(
   A=@n16.G
   B=Multiply(A=@n16.G, B=StaticSwitchParameter(True=@n31, False=@n30) {defaultValue=true, group=Gallium, parameterName=Enable Shoes, sortPriority=1})
   Alpha=ScalarParameter() {group=Gallium, parameterName=LoadingScreenClamp, sortPriority=28}
  )
  n33 = ScalarParameter() {defaultValue=1, parameterName=LoadingScreenEnabled, sliderMax=1}
  n34 = Multiply(
   A=Power(Base=GetMaterialAttributes(None=@n12) {attributeGetTypes [#1]}.Base Color, Exp=@n13) {constExponent=0.25}
   B=LinearInterpolate(
    A=LinearInterpolate(A=@n14.RGB, B=@n15.RGB, Alpha=@n17)
    B=StaticSwitchParameter(True=@n19.RGB, False=@n20.RGB) {group=Gallium, parameterName=OverrideLoadingScreen, sortPriority=4}
    Alpha=Multiply(A=@n32, B=@n33)
   )
  )
  n35 = MaterialFunctionCall() {materialFunction=&MF_CharacterParent_2023_PreFX}
  MaterialFunctionCall(
   Material=BlendMaterialAttributes(
    A=NamedRerouteUsage()
    B=SetMaterialAttributes!(MaterialAttributes=@n11, Base Color=@n34) {attributeSetTypes [#1]}
    Alpha=NamedRerouteUsage()
   ) {vertexAttributeBlendType=UseA}
   M=@n35.M
   Pre Skinned Local Position=@n35.M
   Distance Blend=@n35.M
   Fresnel=@n35.M
  ) {materialFunction=&MF_CharacterParent_2023_PostFX}
  Multiply(A=Divide(A=NamedRerouteUsage()) {constB=100}, B=@n27)
  NamedRerouteDeclaration(None=ComponentMask(None=@n35.Tech Art Mask) {b=true}) {name=TechArtMask.B, nodeColor=<rgba 1 0 0.6562 1>}
  NamedRerouteDeclaration(None=ComponentMask(None=@n10) {b=true}) {name=Input_UserRotation, nodeColor=<rgba 0.8516 1 0 1>}
  n36 = TextureCoordinate() {coordinateIndex=1}
  n37 = Saturate(None=Subtract(A=Floor(None=@n36)))
  n38 = NamedRerouteUsage()
  n39 = Constant2Vector()
  NamedRerouteDeclaration(None=StaticSwitchParameter(True=ComponentMask(None=@n37) {g=true}, False=SphereMask(A=@n38, B=@n39) {attenuationRadius=0.01}) {group=Gallium, parameterName=ShoesUse2ndUV, sortPriority=2}) {name=Mask_Shoes, nodeColor=<rgba 0.0312 1 0 1>}
  n40 = StaticSwitchParameter(True=TextureCoordinate() {coordinateIndex=1}, False=TextureCoordinate()) {group=Gallium, parameterName=ShoesUse2ndUV, sortPriority=2}
  n41 = MaterialFunctionCall(
   X=ScalarParameter() {defaultValue=0.0145, group=Gallium, parameterName=LoadingScreen_UVTransform_LeftShoeUOffset, sortPriority=27}
   Y=ScalarParameter() {defaultValue=-0.349, group=Gallium, parameterName=LoadingScreen_UVTransform_LeftShoeVOffset, sortPriority=27}
  ) {materialFunction=&MakeFloat2}
  n42 = LocalPosition() {includedOffsets=ExcludeOffsets, localOrigin=InstancePreSkinning}
  n43 = Saturate(None=Sign(None=ComponentMask(None=@n42.XYZ) {r=true}))
  NamedRerouteDeclaration(None=LinearInterpolate(A=@n40, B=Add(A=@n40, B=@n41.Result), Alpha=VertexInterpolator(VS=@n43).PS)) {name=UV_Shoes, nodeColor=<rgba 1 0 0.6328 1>}
  NamedRerouteDeclaration(None=@n35.Result) {name=InMA, nodeColor=<rgba 0.4531 1 0 1>}
  n44 = VectorParameter() {group=Gallium, parameterName=LoadingScreen_UVTransform_Front, sortPriority=24, defaultValue=<rgba 0 0 0 1>}
  n45 = AppendVector(A=@n44.RGB, B=@n44.RGB)
  n46 = VectorParameter() {group=Gallium, parameterName=LoadingScreen_UVTransform_Back, sortPriority=25, defaultValue=<rgba 0 0 0 1>}
  n47 = StaticSwitchParameter(True=LinearInterpolate(A=@n45, B=AppendVector(A=@n46.RGB, B=@n46.RGB), Alpha=NamedRerouteUsage()), False=@n45) {defaultValue=true, group=Gallium, parameterName=Enable Rear, sortPriority=1}
  n48 = VectorParameter() {group=Gallium, parameterName=LoadingScreen_UVTransform_Sleeves, sortPriority=26, defaultValue=<rgba 0 0 0 1>}
  n49 = AppendVector(A=@n48.RGB, B=@n48.RGB)
  n50 = Add(
   A=@n49
   B=MaterialFunctionCall(Y=ScalarParameter() {defaultValue=0.125, group=Gallium, parameterName=LoadingScreen_UVTransform_LeftSleeveVOffset, sortPriority=27}) {materialFunction=&MakeFloat4}.Result
  )
  n51 = NamedRerouteUsage()
  n52 = StaticSwitchParameter(True=LinearInterpolate(A=@n47, B=LinearInterpolate(A=@n49, B=@n50, Alpha=@n51), Alpha=NamedRerouteUsage()), False=@n47) {defaultValue=true, group=Gallium, parameterName=Enable Sleeves, sortPriority=1}
  n53 = VectorParameter() {group=Gallium, parameterName=LoadingScreen_UVTransform_Shoes, sortPriority=26, defaultValue=<rgba 0.424 -0.19 0 0.36>}
  n54 = StaticSwitchParameter(True=LinearInterpolate(A=@n52, B=AppendVector(A=@n53.RGB, B=@n53.RGB), Alpha=NamedRerouteUsage()), False=@n52) {defaultValue=true, group=Gallium, parameterName=Enable Shoes, sortPriority=1}
  NamedRerouteDeclaration(None=ComponentMask(None=@n54) {g=true, r=true}) {name=MaskOffset, nodeColor=<rgba 0.5234 0 1 1>}
  NamedRerouteDeclaration(None=ComponentMask(None=@n54) {b=true}) {name=MaskRotation, nodeColor=<rgba 0 0.5781 1 1>}
  NamedRerouteDeclaration(None=ComponentMask(None=@n54) {a=true}) {name=MaskScale, nodeColor=<rgba 0.3828 0 1 1>}
  n55 = NamedRerouteUsage()
  n56 = MaterialFunctionCall(Float2=Saturate(None=Floor(None=@n55))) {materialFunction=&BreakOutFloat2Components}
  NamedRerouteDeclaration(None=@n56.R) {name=Mask_Sleeves, nodeColor=<rgba 0 0.7656 1 1>}
  NamedRerouteDeclaration(None=@n56.G) {name=Mask_Rear, nodeColor=<rgba 1 0.9141 0 1>}
  NamedRerouteDeclaration(None=Constant2Vector()) {name=Input_UserOffset, nodeColor=<rgba 0.8516 1 0 1>}
  NamedRerouteDeclaration(None=Constant() {r=100}) {name=Input_UserScale, nodeColor=<rgba 0.8516 1 0 1>}
  Subtract(A=@n1) {constB=50}
  Subtract(A=@n24) {constB=0.5}
 ]
}
