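# Physics asset for the GalliumReality medium male player body cosmetic, covering secondary motion of belt bag, shirt flaps and waist ties.
# Capsules on pelvis, thighs and accessory anchor bones form the base, with only the belt bag anchor angularly locked to the pelvis.
# Child bag, shirt and tie bodies hang from these anchors with tight angular limits: shirt pieces swing on one axis, ties on two, and the bag swings and twists.
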PhysicsAsset /BRCosmetics/Characters/Player/Male/Medium/Bodies/M_MED_GalliumReality/Meshes/M_MED_GalliumReality_PhysicsFE
bodies {
 base_belt_bag [{shapeType=Capsule, length=1, radius=1, shapeName=base_belt_bag_capsule, rotation=<rot 90 0 0>}]
 base_frontshirt_fliper [{shapeType=Capsule, length=1, radius=1, shapeName=base_frontshirt_fliper_capsule, rotation=<rot 90 0 0>}]
 base_shirt_back [{shapeType=Capsule, length=0.505, radius=0.505, shapeName=base_shirt_back_capsule, rotation=<rot 0 -90 0>}]
 base_shirt_l [{shapeType=Capsule, length=0.505, radius=0.505, shapeName=base_shirt_l_capsule, rotation=<rot 0 -90 0>}]
 base_shirt_r [{shapeType=Capsule, length=0.505, radius=0.505, shapeName=base_shirt_r_capsule, rotation=<rot 0 -90 0>}]
 base_waist_tie_r [{shapeType=Capsule, length=1, radius=1, shapeName=base_waist_tie_r_capsule, rotation=<rot 90 0 0>}]
 dyn_belt_bag [
  {shapeType=Capsule, length=6, radius=2, shapeName=dyn_belt_bag_capsule, center=<7 6.0677 2.6146>, rotation=<rot 90 0 0>}
  {shapeType=Capsule, length=6, radius=2, shapeName=dyn_belt_bag_capsule, center=<7 -4.1932 2.6146>, rotation=<rot 90 0 0>}
 ]
 dyn_shirt_back [{shapeType=Capsule, length=3, radius=3, shapeName=dyn_shirt_back_capsule, center=<2 0 0>, rotation=<rot 90 0 0>}]
 dyn_shirt_l [{shapeType=Capsule, length=3, radius=3, shapeName=dyn_shirt_l_capsule, center=<2 0 0>, rotation=<rot 90 0 0>}]
 dyn_shirt_r [{shapeType=Capsule, length=3, radius=3, shapeName=dyn_shirt_r_capsule, center=<2 0 0>, rotation=<rot 90 0 0>}]
 dyn_waist_tie01_l [{shapeType=Capsule, length=7, radius=2, shapeName=dyn_waist_tie01_l_capsule, center=<8 0 0.3061>, rotation=<rot 90 0 0>}]
 dyn_waist_tie01_r [{shapeType=Capsule, length=7, radius=2, shapeName=dyn_waist_tie01_r_capsule, center=<4.5826 0 0.8586>, rotation=<rot 90 0 0>}]
 dyn_waist_tie_l [{shapeType=Capsule, length=1, radius=1, shapeName=dyn_waist_tie_l_capsule, rotation=<rot 90 0 0>}]
 pelvis [{shapeType=Capsule, length=8, radius=8, shapeName=pelvis_capsule, center=<-15 0 0>}]
 thigh_l [{shapeType=Capsule, length=20, radius=10, shapeName=thigh_l_capsule, center=<-20 0 0>, rotation=<rot 90 0 0>}]
 thigh_r [{shapeType=Capsule, length=20, radius=10, shapeName=thigh_r_capsule, center=<20 0 0>, rotation=<rot 90 0 0>}]
}
constraints [
 {bone1Name=dyn_belt_bag, bone2Name=base_belt_bag, swing1LimitDegrees=0, swing2LimitDegrees=15, twistLimitDegrees=20}
 {
  bone1Name=base_frontshirt_fliper
  bone2Name=pelvis
  swing1LimitDegrees=45
  swing1Motion=Free
  swing2LimitDegrees=45
  swing2Motion=Free
  twistLimitDegrees=45
  twistMotion=Free
 }
 {
  bone1Name=base_belt_bag
  bone2Name=pelvis
  swing1LimitDegrees=45
  swing1Motion=Locked
  swing2LimitDegrees=45
  swing2Motion=Locked
  twistLimitDegrees=45
  twistMotion=Locked
 }
 {
  bone1Name=thigh_l
  bone2Name=pelvis
  swing1LimitDegrees=45
  swing1Motion=Free
  swing2LimitDegrees=45
  swing2Motion=Free
  twistLimitDegrees=45
  twistMotion=Free
 }
 {
  bone1Name=thigh_r
  bone2Name=pelvis
  swing1LimitDegrees=45
  swing1Motion=Free
  swing2LimitDegrees=45
  swing2Motion=Free
  twistLimitDegrees=45
  twistMotion=Free
 }
 {
  bone1Name=base_shirt_back
  bone2Name=pelvis
  swing1LimitDegrees=45
  swing1Motion=Free
  swing2LimitDegrees=45
  swing2Motion=Free
  twistLimitDegrees=45
  twistMotion=Free
 }
 {bone1Name=dyn_shirt_r, bone2Name=base_shirt_r, swing1LimitDegrees=5, swing2LimitDegrees=0, twistLimitDegrees=0}
 {
  bone1Name=base_shirt_r
  bone2Name=pelvis
  swing1LimitDegrees=45
  swing1Motion=Free
  swing2LimitDegrees=45
  swing2Motion=Free
  twistLimitDegrees=45
  twistMotion=Free
 }
 {bone1Name=dyn_shirt_l, bone2Name=base_shirt_l, swing1LimitDegrees=5, swing2LimitDegrees=0, twistLimitDegrees=0}
 {
  bone1Name=base_shirt_l
  bone2Name=pelvis
  swing1LimitDegrees=45
  swing1Motion=Free
  swing2LimitDegrees=45
  swing2Motion=Free
  twistLimitDegrees=45
  twistMotion=Free
 }
 {bone1Name=dyn_shirt_back, bone2Name=base_shirt_back, swing1LimitDegrees=5, swing2LimitDegrees=0, twistLimitDegrees=0}
 {bone1Name=dyn_waist_tie01_l, bone2Name=dyn_waist_tie_l, swing1LimitDegrees=30, swing2LimitDegrees=20, twistLimitDegrees=0}
 {bone1Name=dyn_waist_tie01_r, bone2Name=base_waist_tie_r, swing1LimitDegrees=30, swing2LimitDegrees=30, twistLimitDegrees=0}
 {
  bone1Name=dyn_waist_tie_l
  bone2Name=pelvis
  swing1LimitDegrees=45
  swing1Motion=Free
  swing2LimitDegrees=45
  swing2Motion=Free
  twistLimitDegrees=45
  twistMotion=Free
 }
 {
  bone1Name=base_waist_tie_r
  bone2Name=pelvis
  swing1LimitDegrees=45
  swing1Motion=Free
  swing2LimitDegrees=45
  swing2Motion=Free
  twistLimitDegrees=45
  twistMotion=Free
 }
]
