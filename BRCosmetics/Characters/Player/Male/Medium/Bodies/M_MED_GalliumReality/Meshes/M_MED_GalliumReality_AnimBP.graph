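# Animation Blueprint for the GalliumReality medium male body cosmetic, adding secondary motion to clothing and accessories on the copied parent pose.
# A state-driven Control Rig feeds front-end or in-game RigidBody physics, then AnimDynamics for bracelet, drawstrings, zipper and sleeves.
# Additive bone offsets correct belt, shirt, bag and pant placement last.
# The event graph derives alphas from movement and weapon state and drives WPO cloth material parameters, raising wind while skydiving.
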AnimBlueprint /BRCosmetics/Characters/Player/Male/Medium/Bodies/M_MED_GalliumReality/Meshes/M_MED_GalliumReality_AnimBP
target_skeleton=&/M_MED_GalliumReality_Skeleton
graphs [
 AnimGraph {nodes [
   n1 = "Transform (Modify) Bone - Bone: base_beltbuckle_front"!(
    ComponentPose=Local To Component(LocalPose=Use cached pose 'AnimDynamics'())
    bAlphaBoolEnabled:bool=True
    Translation="2,-1,-1"
    Rotation="0,0,0"
   ) -> (Pose) {
    alphaInputType=Bool
    rotationMode=BMM_Additive
    translationMode=BMM_Additive
    alphaBoolBlend {blendInTime=0.2, blendOption=Linear, blendOutTime=0.2}
    alphaScaleBiasClamp {bInterpResult=true, interpSpeedDecreasing=5}
    boneToModify {boneName=base_beltbuckle_front}
    translation=<2 -1 -1>
   }
   n2 = "Transform (Modify) Bone - Bone: base_shirt_back"!(ComponentPose=@n1, bAlphaBoolEnabled:bool=True, Translation="0,4,3", Rotation="0,0,0") -> (Pose) {
    alphaInputType=Bool
    rotationMode=BMM_Additive
    translationMode=BMM_Additive
    alphaBoolBlend {blendInTime=0.2, blendOption=Linear, blendOutTime=0.2}
    alphaScaleBiasClamp {bInterpResult=true, interpSpeedDecreasing=5}
    boneToModify {boneName=base_shirt_back}
    translation {y=4, z=3}
   }
   n3 = "Transform (Modify) Bone - Bone: base_belt_bag"!(ComponentPose=@n2, bAlphaBoolEnabled:bool=True, Translation="-0.404,-1.000,-0.104", Rotation="-8.234,18.731,-3.952") -> (Pose) {
    alphaInputType=Bool
    rotationMode=BMM_Additive
    translationMode=BMM_Additive
    alphaBoolBlend {blendInTime=0.2, blendOption=Linear, blendOutTime=0.2}
    alphaScaleBiasClamp {bInterpResult=true, interpSpeedDecreasing=5}
    boneToModify {boneName=base_belt_bag}
    rotation=<rot -8.234 18.731 -3.952>
    translation=<-0.404 -1 -0.104>
   }
   n4 = "Transform (Modify) Bone - Bone: base_beltbuckle_front"(ComponentPose=@n3, bAlphaBoolEnabled=True, Translation="0.000,0.724,0.000", Rotation="-0.000,-13.817,0.000") {
    alphaInputType=Bool
    rotationMode=BMM_Additive
    translationMode=BMM_Additive
    alphaBoolBlend {blendInTime=0.2, blendOption=Linear, blendOutTime=0.2}
    alphaScaleBiasClamp {bInterpResult=true, interpSpeedDecreasing=5}
    boneToModify {boneName=base_beltbuckle_front}
    rotation {yaw=-13.817}
    translation {y=0.724}
   }
   n5 = "Transform (Modify) Bone - Bone: base_frontshirt_fliper"!(
    ComponentPose=@n4
    bAlphaBoolEnabled:bool=Get bIsSwimSprinting() -> (bIsSwimSprinting:bool)
    Translation="-6,0,4"
    Rotation="0,0,0"
    Scale="1,1,1"
   ) -> (Pose) {
    alphaInputType=Bool
    translationMode=BMM_Additive
    translationSpace=BCS_BoneSpace
    alphaBoolBlend {blendOption=Linear}
    boneToModify {boneName=base_frontshirt_fliper}
    translation {x=-6, z=4}
   }
   n6 = "Transform (Modify) Bone - Bone: base_waist_tie_r"(
    ComponentPose=@n5
    Alpha=Get UpperArmRot_r() -> (UpperArmRot_r:double)
    Translation="2.530,-2.040,1.811"
    Rotation="27.626,-1.973,-0.899"
    Scale="1,1,1"
   ) {
    rotationMode=BMM_Additive
    rotationSpace=BCS_BoneSpace
    scaleMode=BMM_Additive
    scaleSpace=BCS_BoneSpace
    translationMode=BMM_Additive
    translationSpace=BCS_BoneSpace
    alphaBoolBlend {blendOption=Linear}
    alphaScaleBiasClamp {bInterpResult=true}
    boneToModify {boneName=base_waist_tie_r}
    rotation=<rot 27.626 -1.973 -0.899>
    translation=<2.53 -2.04 1.811>
   }
   n7 = "Transform (Modify) Bone - Bone: base_belt_bag"!(ComponentPose=@n6, Alpha:float=Get isCrouching() -> (isCrouching:double), Translation="0,0,-1", Rotation="0,20,10", Scale:Vector="1,1,1") -> (Pose) {
    rotationMode=BMM_Additive
    rotationSpace=BCS_BoneSpace
    translationMode=BMM_Additive
    translationSpace=BCS_BoneSpace
    alphaBoolBlend {blendOption=Linear}
    alphaScaleBiasClamp {bInterpResult=true, interpSpeedDecreasing=5, interpSpeedIncreasing=2}
    boneToModify {boneName=base_belt_bag}
    rotation {roll=10, yaw=20}
    translation {z=-1}
   }
   n8 = "Transform (Modify) Bone - Bone: base_belt_bag"(
    ComponentPose=@n7
    Alpha=Get jumpmove() -> (jumpmove:double)
    Translation="1.912,-1.524,-0.387"
    Rotation="26.749,16.666,-15.887"
    Scale="1,1,1"
   ) {
    rotationMode=BMM_Additive
    rotationSpace=BCS_BoneSpace
    translationMode=BMM_Additive
    translationSpace=BCS_BoneSpace
    alphaBoolBlend {blendOption=Linear}
    alphaScaleBiasClamp {bInterpResult=true, interpSpeedDecreasing=5, interpSpeedIncreasing=5}
    boneToModify {boneName=base_belt_bag}
    rotation=<rot 26.749 16.666 -15.887>
    translation=<1.912 -1.524 -0.387>
   }
   n9 = "Transform (Modify) Bone - Bone: base_pantleg_r"!(ComponentPose=@n8, bAlphaBoolEnabled:bool=True, Translation="0,0,1", Rotation="0,0,0") -> (Pose) {
    alphaInputType=Bool
    rotationMode=BMM_Additive
    translationMode=BMM_Additive
    alphaBoolBlend {blendInTime=0.2, blendOption=Linear, blendOutTime=0.2}
    alphaScaleBiasClamp {bInterpResult=true, interpSpeedDecreasing=5}
    boneToModify {boneName=base_pantleg_r}
    translation {z=1}
   }
   n10 = "Transform (Modify) Bone - Bone: base_frontshirt_fliper"!(ComponentPose=@n9, bAlphaBoolEnabled=True, Translation="-3.873,8.000,0.000", Rotation="-20.000,10.000,-0.000") -> (Pose) {
    alphaInputType=Bool
    rotationMode=BMM_Additive
    translationMode=BMM_Additive
    alphaBoolBlend {blendInTime=0.2, blendOption=Linear, blendOutTime=0.2}
    alphaScaleBiasClamp {bInterpResult=true, interpSpeedDecreasing=5}
    boneToModify {boneName=base_frontshirt_fliper}
    rotation=<rot -20 10 0>
    translation=<-3.873 8>
   }
   n11 = "Transform (Modify) Bone - Bone: base_shirt_back"(ComponentPose=@n10, bAlphaBoolEnabled=True, Translation="-8.676,-0.000,-4.226", Rotation="40.000,0.000,0.000") {
    alphaInputType=Bool
    rotationMode=BMM_Additive
    translationMode=BMM_Additive
    alphaBoolBlend {blendInTime=0.2, blendOption=Linear, blendOutTime=0.2}
    alphaScaleBiasClamp {bInterpResult=true, interpSpeedDecreasing=5}
    boneToModify {boneName=base_shirt_back}
    rotation {pitch=40}
    translation {x=-8.676, z=-4.226}
   }
   Output Pose(Result=Component To Local(ComponentPose=@n11)) {layerGroup=None}
   n12 = Control Rig!(
    Source=Copy Pose From Mesh() {bUseAttachedParent=true}
    Alpha
    JumpAlpha:double=Get IsJumping() -> (IsJumping:double)
    crouchwalkAlpha:double=Get isCrouching()
    TargetingAlpha:double=Get isTargeting() -> (isTargeting:double)
    KneeSlideAlpha:double=Get isKneeSlide() -> (isKneeSlide:double)
    BaseballSlidingAlpha:double=Get isBasball sliding() -> (isBasball sliding:double)
    GunOutAlpha:double=Get Gun_out() -> (Gun_out:double)
    emoteAlpha:double=Get isPlaying Emote() -> (isPlaying Emote:double)
    is crouching:double=Get iscrounchingStill() -> (iscrounchingStill:double)
    is Melee:double=Get isMelee() -> (isMelee:double)
   ) -> (Pose) {
    alphaBoolBlend {blendOption=Linear}
    controlRigAssetReference {blueprintRigClass=&/M_MED_GalliumReality_CtrlRig.M_MED_GalliumReality_CtrlRig_C}
   }
   Save cached pose 'Control Rig'(Pose=@n12) {cacheName=Control Rig}
   n13 = RigidBody!(ComponentPose=Local To Component(LocalPose=Use cached pose 'Control Rig'()), Alpha, OverrideWorldGravity:Vector="0,0,-980", ExternalForce) -> (Pose) {
    bClampLinearTranslationLimitToRefPose=true
    bOverrideWorldGravity=true
    overridePhysicsAsset=&/M_MED_GalliumReality_PhysicsFE
    simulationSpace=BaseBoneSpace
    alphaBoolBlend {blendOption=Linear}
    baseBoneRef {boneName=spine_01}
    componentAppliedLinearAccClamp=<2000 2000 1000>
    componentLinearAccScale {x=0.4}
    componentLinearVelScale {x=0.4}
    overrideWorldGravity {z=-980}
    simSpaceSettings {maxAngularAcceleration=100, maxAngularVelocity=5, maxLinearAcceleration=300, maxLinearVelocity=50, velocityScaleZ=0, worldAlpha=1}
   }
   n14 = RigidBody(ComponentPose=Local To Component(LocalPose=Use cached pose 'Control Rig'()), OverrideWorldGravity="0,0,-980") {
    bClampLinearTranslationLimitToRefPose=true
    bOverrideWorldGravity=true
    overridePhysicsAsset=&/M_MED_GalliumReality_Physics
    simulationSpace=BaseBoneSpace
    alphaBoolBlend {blendOption=Linear}
    baseBoneRef {boneName=spine_01}
    componentAppliedLinearAccClamp=<2000 2000 1000>
    componentLinearAccScale {x=0.4}
    componentLinearVelScale {x=0.4}
    overrideWorldGravity {z=-980}
    simSpaceSettings {maxAngularAcceleration=100, maxAngularVelocity=5, maxLinearAcceleration=300, maxLinearVelocity=50, velocityScaleZ=0, worldAlpha=1}
   }
   Save cached pose 'RigidBodies'(Pose=Blend Poses by bool(
     BlendPose_0=Component To Local(ComponentPose=@n13)
     BlendPose_1=Component To Local(ComponentPose=@n14)
     BlendTime_0=0.2
     bActiveValue=Get bIsInFrontEnd()
    ) {blendType=Linear}) {cacheName=RigidBodies}
   n15 = Anim Dynamics - Bone: base_bracelet(ComponentPose=Local To Component(LocalPose=Use cached pose 'RigidBodies'()), Alpha=1, ExternalForce="0,0,0") {
    angularDampingOverride=1
    angularSpringConstant=4000
    bAngularSpring=true
    bOverrideAngularDamping=true
    bUseGravityOverride=true
    gravityScale=0.35
    simulationSpace=BoneRelative
    alphaBoolBlend {blendOption=Linear}
    boundBone {boneName=base_bracelet}
    componentAppliedLinearAccClamp=<10000 10000 10000>
    componentLinearAccScale=<0.2 0.2 0.2>
    componentLinearVelScale=<0.2 0.2 0.2>
    gravityOverride {z=-980}
    physicsBodyDefinitions [{
      boundBone {boneName=base_bracelet}
      boxExtents=<10 20 20>
      constraintSetup {angularLimitsMax=<5 5 5>, angularLimitsMin=<-5 -5 -5>, angularTarget=<1 0 0>}
      localJointOffset=<3 0 0>
     }]
   }
   n16 = Anim Dynamics - Bone: dyn_drawstring_l(ComponentPose=@n15, Alpha=1, ExternalForce="0,0,0") {
    angularDampingOverride=1
    angularSpringConstant=5000
    bAngularSpring=true
    bOverrideAngularDamping=true
    bUseGravityOverride=true
    gravityScale=0.35
    simulationSpace=RootRelative
    alphaBoolBlend {blendOption=Linear}
    boundBone {boneName=dyn_drawstring_l}
    componentLinearAccScale=<1 1 1>
    componentLinearVelScale=<1 1 1>
    gravityOverride {z=-980}
    physicsBodyDefinitions [{
      boundBone {boneName=dyn_drawstring_l}
      boxExtents=<20 20 20>
      constraintSetup {angularLimitsMax=<15 15 15>, angularLimitsMin=<-15 -15 -15>, angularTarget=<1 0 0>}
      localJointOffset=<4 0 0>
     }]
   }
   n17 = Anim Dynamics - Bone: dyn_drawstring_r(ComponentPose=@n16, Alpha=1, ExternalForce="0,0,0") {
    angularDampingOverride=1
    angularSpringConstant=5000
    bAngularSpring=true
    bOverrideAngularDamping=true
    bUseGravityOverride=true
    gravityScale=0.35
    simulationSpace=RootRelative
    alphaBoolBlend {blendOption=Linear}
    boundBone {boneName=dyn_drawstring_r}
    componentLinearAccScale=<1 1 1>
    componentLinearVelScale=<1 1 1>
    gravityOverride {z=-980}
    physicsBodyDefinitions [{
      boundBone {boneName=dyn_drawstring_r}
      boxExtents=<20 20 20>
      constraintSetup {angularLimitsMax=<15 15 15>, angularLimitsMin=<-15 -15 -15>, angularTarget=<1 0 0>}
      localJointOffset=<4 0 0>
     }]
   }
   n18 = Anim Dynamics - Bone: zipper(ComponentPose=@n17, Alpha=1, ExternalForce="0,0,0") {
    angularDampingOverride=1
    angularSpringConstant=5000
    bAngularSpring=true
    bOverrideAngularDamping=true
    bUseGravityOverride=true
    gravityScale=0.35
    simulationSpace=RootRelative
    alphaBoolBlend {blendOption=Linear}
    boundBone {boneName=zipper}
    componentAppliedLinearAccClamp=<50000 50000 50000>
    componentLinearAccScale=<1 1 1>
    componentLinearVelScale=<1 1 1>
    gravityOverride {z=-980}
    physicsBodyDefinitions [{
      boundBone {boneName=zipper}
      boxExtents=<10 20 20>
      constraintSetup {angularLimitsMax=<0 -10 40>, angularLimitsMin=<0 -50 -40>, angularTarget=<1 0 0>}
      localJointOffset=<4 0 0>
     }]
   }
   n19 = Get Dynoff() -> (Dynoff:double)
   n20 = Anim Dynamics - Bone: dyn_sleevegiggler!(ComponentPose=@n18, Alpha=@n19, GravityOverride:Vector="0,0,0", ExternalForce="0,0,0") -> (Pose) {
    angularDampingOverride=1
    angularSpringConstant=1000
    bOverrideAngularDamping=true
    bUseGravityOverride=true
    gravityScale=0.35
    simulationSpace=RootRelative
    alphaBoolBlend {blendOption=Linear}
    alphaScaleBiasClamp {bInterpResult=true, interpSpeedDecreasing=2, interpSpeedIncreasing=2}
    boundBone {boneName=dyn_sleevegiggler}
    componentAppliedLinearAccClamp=<10000 10000 10000>
    componentLinearAccScale=<0.1 0.1 0.1>
    componentLinearVelScale=<0.1 0.1 0.1>
    gravityOverride {z=-980}
    physicsBodyDefinitions [{
      boundBone {boneName=dyn_sleevegiggler}
      boxExtents=<10 20 20>
      constraintSetup {angularLimitsMax=<0 5 5>, angularLimitsMin=<0 -5 -2>, angularTarget=<1 0 0>}
      localJointOffset=<8 0 0>
     }]
    relativeSpaceBone {boneName=upperarm_l}
   }
   n21 = Anim Dynamics - Bone: dyn_sleevegiggler1!(ComponentPose=@n20, Alpha=@n19, GravityOverride:Vector="0,0,0", ExternalForce="0,0,0") -> (Pose) {
    angularDampingOverride=0.9999
    angularSpringConstant=1000
    bOverrideAngularDamping=true
    bUseGravityOverride=true
    gravityScale=0.35
    simulationSpace=RootRelative
    alphaBoolBlend {blendOption=Linear}
    alphaScaleBiasClamp {bInterpResult=true, interpSpeedDecreasing=2, interpSpeedIncreasing=2}
    boundBone {boneName=dyn_sleevegiggler1}
    componentAppliedLinearAccClamp=<10000 10000 10000>
    componentLinearAccScale=<0.5 0.5 0.5>
    componentLinearVelScale=<0.5 0.5 0.5>
    gravityOverride {z=-980}
    physicsBodyDefinitions [{
      boundBone {boneName=dyn_sleevegiggler1}
      boxExtents=<10 20 20>
      constraintSetup {angularLimitsMax=<0 7 5>, angularLimitsMin=<0 -7 -2>, angularTarget=<1 0 0>}
      localJointOffset=<8 0 0>
     }]
    relativeSpaceBone {boneName=upperarm_r}
   }
   n22 = Get Dynon() -> (Dynon:double)
   n23 = Anim Dynamics - Bone: dyn_sleevegiggler(ComponentPose=@n21, Alpha=@n22, GravityOverride="0,0,0", ExternalForce="0,0,0") {
    angularDampingOverride=1
    angularSpringConstant=1000
    gravityScale=0
    simulationSpace=BoneRelative
    alphaBoolBlend {blendOption=Linear}
    alphaScaleBiasClamp {bInterpResult=true, interpSpeedDecreasing=2, interpSpeedIncreasing=2}
    boundBone {boneName=dyn_sleevegiggler}
    componentAppliedLinearAccClamp=<0 0 0>
    gravityOverride {z=-980}
    physicsBodyDefinitions [{
      boundBone {boneName=dyn_sleevegiggler}
      boxExtents=<1 1 1>
      constraintSetup {angularLimitsMax=<0 0 0>, angularLimitsMin=<0 0 0>, angularTarget=<1 0 0>}
      localJointOffset=<8 0 0>
     }]
    relativeSpaceBone {boneName=upperarm_l}
   }
   n24 = Anim Dynamics - Bone: dyn_sleevegiggler1(ComponentPose=@n23, Alpha=@n22, GravityOverride="0,0,0", ExternalForce="0,0,0") {
    angularDampingOverride=0.9999
    angularSpringConstant=1000
    gravityScale=0
    simulationSpace=BoneRelative
    alphaBoolBlend {blendOption=Linear}
    alphaScaleBiasClamp {bInterpResult=true, interpSpeedDecreasing=2, interpSpeedIncreasing=2}
    boundBone {boneName=dyn_sleevegiggler1}
    componentAppliedLinearAccClamp=<0 0 0>
    gravityOverride {z=-980}
    physicsBodyDefinitions [{
      boundBone {boneName=dyn_sleevegiggler1}
      boxExtents=<1 1 1>
      constraintSetup {angularLimitsMax=<0 0 0>, angularLimitsMin=<0 0 0>, angularTarget=<1 0 0>}
      localJointOffset=<8 0 0>
     }]
    relativeSpaceBone {boneName=upperarm_r}
   }
   Save cached pose 'AnimDynamics'(Pose=Component To Local(ComponentPose=@n24)) {cacheName=AnimDynamics}
  ]}
 EventGraph {nodes [
   n1 = Event BlueprintUpdateAnimation()
   > n2 = Sequence!(execute) -> (then_0, then_1, then_2:Exec, then_3:Exec)
   > n3 = Gravity Override State Function DEPRECATED(CharacterPartAnimInstance=Self-Reference(), GameTimeInSeconds=GetGameTimeInSeconds(), GravityOverrideParams=Get Gravity Override Params())
   > Set GravityOverride(GravityOverride=@n3)
   n4 = Get bIsMoving2D()
   n5 = Get jumpbool() -> (jumpbool:bool)
   > Set jumpmove(jumpmove:double="To Float (Boolean)"(InBool=AND Boolean(A=@n5, B=@n4))) -> (Output_Get:double)
   n6 = Get bIsCrouching()
   n7 = Get bIsMoving2D()
   > Set isCrouching(isCrouching:double="To Float (Boolean)"(InBool=AND Boolean(A=@n6, B=@n7))) -> (Output_Get:double)
   > Set isTargeting(isTargeting:double="To Float (Boolean)"(InBool=Get bIsTargeting())) -> (Output_Get:double)
   > Set isKneeSlide(isKneeSlide:double="To Float (Boolean)"(InBool=Get bIsKneeSliding())) -> (Output_Get:double)
   n8 = NOT Boolean(A=Get bIsSkyDiving())
   n9 = NOT Boolean(A=Get bIsPlayingMeleeAnim())
   > Set isNOTSkydiving(isNOTSkydiving:double="To Float (Boolean)"(InBool=OR Boolean(A=@n9, B=@n8))) -> (Output_Get:double)
   > Set isBasball sliding(isBasball sliding:double="To Float (Boolean)"(InBool=Get bIsBaseballSliding())) -> (Output_Get:double)
   > Set isPlaying Emote(isPlaying Emote:double="To Float (Boolean)"(InBool=Get bIsPlayingEmote())) -> (Output_Get:double)
   n10 = Get bIsPlayingEmote()
   > Set isPlaying Emote_0(isPlaying Emote_0:double="To Float (Boolean)"(InBool=NOT Boolean(A=@n10))) -> (Output_Get:double)
   n11 = NOT Boolean(A=Get bIsMoving2D())
   n12 = Get bIsCrouching()
   > Set iscrounchingStill(iscrounchingStill:double="To Float (Boolean)"(InBool=AND Boolean(A=@n12, B=@n11))) -> (Output_Get:double)
   n13 = NOT Boolean(A=Get bIsMoving2D())
   n14 = Get bIsPlayingMeleeAnim()
   n15 = Get bIsCrouching()
   > Set isMelee(isMelee:double="To Float (Boolean)"(InBool=AND Boolean!(A=@n14, B=@n15, C:bool=@n13) -> (ReturnValue))) -> (Output_Get:double)
   n16 = NOT Boolean(A=Get bIsPlayingMeleeAnim())
   n17 = NOT Boolean(A=Get bIsParachuteOpened())
   > Set Dynoff(Dynoff:double="To Float (Boolean)"(InBool=OR Boolean(A=@n17, B=@n16))) -> (Output_Get:double)
   n18 = Get bIsPlayingMeleeAnim()
   n19 = Get bIsParachuteOpened()
   > Set Dynon(Dynon:double="To Float (Boolean)"(InBool=OR Boolean(A=@n19, B=@n18))) -> (Output_Get:double)
   n20 = GetOwnerSkeletalMeshComponent()
   n21 = BreakRotator(InRot=BreakTransform(InTransform=GetDeltaTransformFromRefPose(self=@n20, BoneName=spine_05, BaseName=pelvis)).Rotation)
   n22 = Branch(execute=@n2.then_3, Condition=AND Boolean!(A="float > float"(A=@n21, B=20), B=Get bIsCrouching()) -> (ReturnValue))
   > Set UpperArmRot_r(UpperArmRot_r:double=1) -> (Output_Get:double)
   Set UpperArmRot_r(execute=@n22.else, UpperArmRot_r=0)
   n23 = Cast To FortPlayerAnimInstance(execute=@n2.then_2, Object=GetAnimInstance(self=GetOwnerSkeletalMeshComponent()))
   > Set firingPosition(firingPosition:int=0) -> (Output_Get:int=0)
   > n24 = Switch on EFortWeaponCoreAnimation!(execute, Selection:EFortWeaponCoreAnimation Enum=Get WeaponCoreAnim(self=@n23)) -> (
    Melee:Exec
    Pistol:Exec
    Shotgun:Exec
    PaperBlueprint:Exec
    Rifle:Exec
    MeleeOneHand:Exec
    MachinePistol:Exec
    RocketLauncher:Exec
    GrenadeLauncher:Exec
    GoingCommando:Exec
    AssaultRifle:Exec
    TacticalShotgun:Exec
    SniperRifle:Exec
    TrapPlacement:Exec
    ShoulderLauncher:Exec
    AbilityDecoTool:Exec
    Crossbow:Exec
    C4:Exec
    RemoteControl:Exec
    DualWield:Exec
    AR_BullPup:Exec
    AR_ForwardGrip:Exec
    MedPackPaddles:Exec
    SMG_P90:Exec
    AR_DrumGun:Exec
    Consumable_Small:Exec
    Consumable_Large:Exec
    Balloon:Exec
    MountedTurret:Exec
    CreativeTool:Exec
    ExplosiveBow:Exec
    AshtonIndigo:Exec
    AshtonChicago:Exec
    MeleeDualWield:Exec
    Unarmed:Exec
   )
   n25 = Branch(execute=[
     @n24.AR_BullPup
     @n24.AR_DrumGun
     @n24.AR_ForwardGrip
     @n24.AshtonChicago
     @n24.AshtonIndigo
     @n24.AssaultRifle
     @n24.Crossbow
     @n24.DualWield
     @n24.ExplosiveBow
     @n24.GoingCommando
     @n24.GrenadeLauncher
     @n24.MachinePistol
     @n24.Pistol
     @n24.Rifle
     @n24.RocketLauncher
     @n24.SMG_P90
     @n24.Shotgun
     @n24.ShoulderLauncher
     @n24.SniperRifle
     @n24.TacticalShotgun
    ])
   > n26 = Set firingPosition(firingPosition=1) -> (Output_Get=0)
   n27 = NOT Boolean(A=Get bIsCrouching())
   n28 = "To Boolean (Integer)"!(InInt:int=@n26) -> (ReturnValue:bool)
   > Set Gun_out(Gun_out:double="To Float (Boolean)"(InBool=AND Boolean(A=@n28, B=@n27))) -> (Output_Get:double)
   Set firingPosition(execute=@n25.else, firingPosition=0) -> (Output_Get=0)
   Event OnCharacterCustomizationLoaded()
   > n29 = Set SkelMesh(SkelMesh=GetPartSkeletalMeshComponent())
   > n30 = Get Dynamic Material Instance(SkeletalMeshComponent=@n29)
   > Set Mat Instance Dynamic(Mat Instance Dynamic=@n30)
   > n31 = GetOrCreateDynamicOverlayMaterial(MeshComponent=@n29)
   > Set Overlay Mat Instance Dynamic(Overlay Mat Instance Dynamic=@n31)
   Event BlueprintInitializeAnimation()
   > Set SkelMesh(SkelMesh=GetOwnerSkeletalMeshComponent())
   n32 = TryGetPawnOwner()
   > Is Valid(InputObject=@n32)
   > n33 = Cast To FortPlayerPawnAthena(Object=@n32)
   > n34 = Set PlayerPawn(PlayerPawn:Fort Player Pawn Athena Object Reference=@n33) -> (Output_Get:Fort Player Pawn Athena Object Reference)
   > Set Body_Mesh(Body_Mesh=GetSkeletalMeshForPartType(self=@n34))
   n35 = Event OnEndJumping()
   Set isJumping(execute=[@n35.then, @n37.then], isJumping:double=0) -> (Output_Get:double)
   > Set jumpbool(jumpbool:bool=false) -> (Output_Get:bool)
   n36 = Event OnBeginFalling()
   Set isJumping(execute=[@n36.then, @n38.then], isJumping=1)
   > Set jumpbool(jumpbool=true)
   n37 = Event OnEndFalling()
   n38 = Event OnBeginJumping()
   Is Valid(InputObject=Get PlayerPawn() -> (PlayerPawn:Fort Player Pawn Athena Object Reference))
   > Set WPO Globals Function(
    Material=Get Mat Instance Dynamic()
    Overlay Material=Get Overlay Mat Instance Dynamic()
    Skeletal Mesh=Get SkelMesh()
    "Local Noise Transform  Location"=Get Local Noise Transform Location() -> (Local Noise Transform Location:Vector)
    Local Noise Offset=Get Local Noise Offset()
   )
   > Set WPO Regional Function!(
    execute
    Material:Material Instance Dynamic Object Reference=Get Mat Instance Dynamic()
    Overlay Material:Material Instance Dynamic Object Reference=Get Overlay Mat Instance Dynamic()
    Skeletal Mesh:Skeletal Mesh Component Object Reference=Get SkelMesh()
    Velocity Transform:Name=foot_r
    Velocity Interp Speed:double=Get Velocity Interp Speed()
    Collision Transform:Name=calf_r
    Collision Transform Axis:EAxis Enum
    Invert Collision Transform Axis:bool=false
    Delta Time X:double=@n1.DeltaTimeX
    Previous Velocity Transform Location:"Vector (by ref)"=Get R1 Velocity Transform Location()
    Previous Velocity Transform Velocity:"Vector (by ref)"=Get R1 Velocity Transform Velocity()
    Region:int=1
   ) -> (then)
   > Set WPO Regional Function(
    Material=Get Mat Instance Dynamic()
    Overlay Material=Get Overlay Mat Instance Dynamic()
    Skeletal Mesh=Get SkelMesh()
    Velocity Transform=foot_l
    Velocity Interp Speed=Get Velocity Interp Speed()
    Collision Transform=calf_l
    Invert Collision Transform Axis=true
    Delta Time X=@n1.DeltaTimeX
    Previous Velocity Transform Location=Get R2 Velocity Transform Location()
    Previous Velocity Transform Velocity=Get R2 Velocity Transform Velocity()
    Region=2
   )
   > n39 = Branch(Condition=Get bIsSkyDiving())
   > Set WPO Pho Wind Intensity(WPO Pho Wind Intensity=2)
   > n40 = Set WPO Pho Turbulence(WPO Pho Turbulence:double=1) -> (Output_Get:double)
   n41 = Get Mat Instance Dynamic()
   n42 = Get Overlay Mat Instance Dynamic()
   SetScalarParameterValue(execute=[@n40.then, @n44.then], self=[@n41, @n42], ParameterName=ClothSim_WindStrength, Value=Get WPO Pho Wind Intensity())
   n43 = Get WPO Pho Turbulence() -> (WPO Pho Turbulence:double)
   > SetScalarParameterValue(self=[@n42, @n41], ParameterName=ClothSim_R1_TurbulenceStrength, Value=@n43)
   > SetScalarParameterValue(self=[@n42, @n41], ParameterName=ClothSim_R2_TurbulenceStrength, Value=@n43)
   Set WPO Pho Wind Intensity(execute=@n39.else, WPO Pho Wind Intensity=0)
   > n44 = Set WPO Pho Turbulence(WPO Pho Turbulence=0.2)
   SetScalarParameterValue(self=[@n42, @n41], ParameterName=ClothSim_R3_TurbulenceStrength, Value=@n43)
  ]}
]
variables [
 Body_Mesh {type=SkeletalMeshComponent}
 Dynoff {type=number}
 Dynon {type=number}
 firingPosition {type=integer}
 Gravity Override Params {type=GravityOverrideParamsStruct}
 GravityOverride {type=Vector}
 Gun_out {type=number}
 isBasball sliding {type=number}
 isCrouching {type=number}
 iscrounchingStill {type=number}
 isCustomEmotePlaying {type=number}
 IsJumping {type=number}
 isKneeSlide {type=number}
 isMelee {type=number}
 isNOTSkydiving {type=number}
 isPlaying Emote {type=number}
 isPlaying Emote_0 {type=number}
 isTargeting {type=number}
 jumpbool {type=boolean}
 jumpmove {type=number}
 Local Noise Offset {type=Vector}
 Local Noise Transform Location {type=Vector}
 Mat Instance Dynamic {type=MaterialInstanceDynamic}
 nOGun {type=number}
 Overlay Mat Instance Dynamic {type=MaterialInstanceDynamic}
 PlayerPawn {type=FortPlayerPawnAthena}
 R1 Velocity Transform Location {type=Vector}
 R1 Velocity Transform Velocity {type=Vector}
 R2 Velocity Transform Location {type=Vector}
 R2 Velocity Transform Velocity {type=Vector}
 SkelMesh {type=SkeletalMeshComponent}
 UpperArmRot_r {type=number}
 Velocity Interp Speed {type=number}
 WPO Pho Turbulence {type=number}
 WPO Pho Wind Intensity {type=number}
]
